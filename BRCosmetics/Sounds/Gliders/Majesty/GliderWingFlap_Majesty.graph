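# Sound cue for the wing flap of the Majesty glider.
# A Random node picks one of three equally weighted flap waves, then a Modulator applies slight random pitch variation.
# The cue plays quietly in the glider landing sound class with glider thrust attenuation.
# It overrides concurrency to cap simultaneous instances, alongside a reference to the glider open/close concurrency set.
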SoundCue /BRCosmetics/Sounds/Gliders/Majesty/GliderWingFlap_Majesty
root=Modulator_0
nodes [
 {id=Modulator_0, children [Random_1], properties {pitchMax=1.1, pitchMin=0.9}}
 {id=Random_1, children [WavePlayer_1,WavePlayer_2,WavePlayer_4], properties {weights [1,1,1]}}
 {id=WavePlayer_1, properties {soundWaveAssetPtr=&/Glider_Majesty_WingFlap_01}}
 {id=WavePlayer_2, properties {soundWaveAssetPtr=&/Glider_Majesty_WingFlap_02}}
 {id=WavePlayer_4, properties {soundWaveAssetPtr=&/Glider_Majesty_WingFlap_03}}
]
properties {
 attenuationSettings=&Fort_Player_Glider_Thrust_Athena_Att
 bOverrideConcurrency=true
 maxDistance=12500
 soundClassObject=&Fort_Gliders_Land
 volumeMultiplier=0.2
 concurrencyOverrides {maxCount=3}
 concurrencySet [&Athena_GliderOpenClose_Conc]
}
